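; Chrome chest gameplay ability that overrides the search loot tier group on its avatar building container.
; On activation, if the container is unsearched, it asks the game state's ChromeChestGameStateComponent for override info.
; A valid result applies the chrome tier group through SetSearchLootTierGroup, while an invalid one writes the returned group straight to the variable.
; A failed passive activation is written to the shipping log.
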(event Ability|EventFailedToActivatePassiveAbility
  (params (ActorInfo :type Gameplay%20Ability%20Actor%20Info%20Structure))
  ;[0.01003 0.1253 0.151] FailedToActivatePassiveAbility: Log the failure.
  (Fortnite|FortShippingLog self "GA_Athena_ChromeChest: FailedToActivatePassiveAbility" true))

(event Ability|EventActivateAbility
  (bind _returnvalue (Fortnite|Athena|GameState|GetGameStateAthena))
  (bind _returnvalue_2 (Actor|GetComponentbyClass _returnvalue "/ChromeChestGameplay/Gameplay/ChromeChestGameStateComponent.ChromeChestGameStateComponent_C"))
  ;[0 0.2083 0.04224] ActivateAbility: Update to the chrome loot tier group and add the chrome chest gameplay cue tag to the actor.
  (bind _asbuilding_container (Utilities|Casting|CastToBuildingContainer (Ability|GetAvatarActorfromActorInfo)))
  (if (not (Container|IsSearched _asbuilding_container))
    (Utilities|IsValid _returnvalue
      (:Is%20Valid
        (Utilities|IsValid _returnvalue_2
          (:Is%20Valid
            (bind (_bvalid _chromeloottiergroup) (Class|ChromeChestGameStateComponent|GetChromeChestOverrideInfo _returnvalue_2 _asbuilding_container))
            (if _bvalid
              (Container|SetSearchLootTierGroup _asbuilding_container _chromeloottiergroup false true)
              (Variables|/Script/FortniteGame.BuildingContainer|SetSearchLootTierGroup _chromeloottiergroup _asbuilding_container))))))))
